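; Cluster bomb bow arrow projectile that, on a fully charged impact, spawns secondary bomb projectiles, the first immediately and the rest on a looping timer.
; The first launches along the impact normal tilted around the arrow's right vector; later ones fan alternately to either side around world Z.
; Scalable float rows drive count, delay, angles and random speed, which is scaled for pawn hits.
; Bombs inherit the arrow's explosion effects, and the last is flagged.
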(event EventBeginPlay
  (|CallParentFunction) ; This node is disabled and will not be called.\nDrag off pins to build functionality.
)

(event Collision|EventActorBeginOverlap
  (params (OtherActor :type Actor%20Object%20Reference))
  (|CallParentFunction OtherActor) ; This node is disabled and will not be called.\nDrag off pins to build functionality.
)

(event EventTick
  (params (DeltaSeconds :type Float%20%28single-precision%29)))

(event EventOnFullyChargedImpact
  (params (Hit%20Result :type Hit%20Result%20Structure))
  (bind _output_get (Variables|Default|SetCachedHit Hit%20Result))
  ;[1 1 1] Check if we hit a player and set the flag so that special case can be handled
  (Utilities|Casting|CastToFortPlayerPawn (.HitActor (Collision|BreakHitResult _output_get))
    (:then
      (Variables|Default|SetbHitPlayer true))
    (:CastFailed
      _))
  (bind _returnvalue (Utilities|Time|SetTimerbyEvent (delegate Custom|DelayedClusterBombSpawn) (ScalableFloat|GetValueatLevel (Variables|Rows|GetRow_DelayBetweenSpawns)) true))
  (Variables|Default|SetBombSpawnTimerHandle _returnvalue)
  (CallFunction|SpawnClusterBomb (Variables|Default|GetCachedHit)))

(event Custom|DelayedClusterBombSpawn
  (CallFunction|SpawnClusterBomb (Variables|Default|GetCachedHit))
  (if (CallFunction|CheckAllBombsSpawned)
    (Utilities|Time|ClearandInvalidateTimerbyHandle (Variables|Default|GetBombSpawnTimerHandle))))

(fn CheckAllBombsSpawned
  (return (>= (Variables|Default|GetNumBombsSpawned) (Math|Float|Truncate (ScalableFloat|GetValueatLevel (Variables|Rows|GetRow_NumBombs))))))

(fn GetAuxiliarySpawnDirectionFromIndex
  (params (Index :type Integer))
  (bind _returnvalue (Math|Integer|%25%28Integer%29 Index 2))
  (bind _returnvalue_2 (== _returnvalue 0))
  (bind _index (Variables|GetIndex))
  ;[1 1 1] Determine the sign of the angle. Even indices will have negative angles (rotated to the left of the center bomb) and odd indices will have positive angles (rotated to the right of the center bomb)
  (LocalVariables|SetAngleSign (Math|Integer|SelectInt -1 1 _returnvalue_2))
  (if _returnvalue_2
    ;[1 1 1] This technically isn't necessary for only three spawned bombs, but calculates a multiplier for the bomb's angle - e.g. the 4th bomb (index 3) should be two intervals to the right of the center bomb, so its angle will be AngleBetweenBombs * sign * 2
    (LocalVariables|SetOffset (/ _index 2))
    (else
      (LocalVariables|SetOffset (+ (/ (- _index 1) 2) 1))))
  (return (Math|Vector|RotateVectorAroundAxis (Variables|Default|GetInitialSpawnDirection) (* (* (ScalableFloat|GetValueatLevel (Variables|Rows|GetRow_AngleBetweenBombs)) (Math|Conversions|ToFloat(Integer) (LocalVariables|GetAngleSign))) (Math|Conversions|ToFloat(Integer) (LocalVariables|GetOffset))) "0, 0,1.000000")))

(fn GetBombSpeedFromRange
  ;[1 1 1] Chooses a random speed within the specified range. If we hit a player, the speed is reduced by multiplying it by a random scalar in the range of PlayerHitBombSpeedScalar
  (return (select (Variables|Default|GetbHitPlayer) (* (Math|Random|RandomFloatinRange (ScalableFloat|GetValueatLevel (Variables|Rows|GetRow_BombSpeedMin)) (ScalableFloat|GetValueatLevel (Variables|Rows|GetRow_BombSpeedMax))) (Math|Random|RandomFloatinRange (ScalableFloat|GetValueatLevel (Variables|Rows|GetRow_PlayerHitSpeedScalarMin)) (ScalableFloat|GetValueatLevel (Variables|Rows|GetRow_PlayerHitSpeedScalarMax)))) (Math|Random|RandomFloatinRange (ScalableFloat|GetValueatLevel (Variables|Rows|GetRow_BombSpeedMin)) (ScalableFloat|GetValueatLevel (Variables|Rows|GetRow_BombSpeedMax))))))

(fn GetInitialSpawnDirection
  (params (Normal :type Vector))
  ;[1 1 1] Rotate the impact normal around the arrow's right vector. Using a positive angle will tilt the bomb in the direction the player fired (which is what we want)
  (return (Math|Vector|RotateVectorAroundAxis Normal (ScalableFloat|GetValueatLevel (Variables|Rows|GetRow_LaunchAngle)) (Transformation|GetActorRightVector))))

(fn GetSpawnDirectionAsRotation
  (params (Origin :type Vector) (Direction :type Vector))
  (return (Math|Rotator|FindLookatRotation Origin (+ Origin Direction))))

(fn SpawnClusterBomb
  (params (Hit :type Hit%20Result%20Structure))
  (bind (_bblockinghit _binitialoverlap _time _distance _location _impactpoint _normal) (Collision|BreakHitResult Hit))
  ;[1 1 1] Check the index of the bomb being spawned. If it's the very first bomb, it should be in the center, so we calculate its direction by rotating the impact normal by the specified angle around the arrow's right vector (see GetInitialSpawnDirection). Otherwise, the bombs are to the side of the initial bomb, so we calculate its direction based on the initial bomb's direction
  (if (== (Variables|Default|GetNumBombsSpawned) 0)
    (bind _direction (CallFunction|GetInitialSpawnDirection _normal))
    (bind _output_get (Variables|Default|SetInitialSpawnDirection _direction))
    (LocalVariables|SetSpawnDirection _output_get)
    (else
      (bind _direction_2 (CallFunction|GetAuxiliarySpawnDirectionfromIndex (Variables|Default|GetNumBombsSpawned)))
      (LocalVariables|SetSpawnDirection _direction_2)))
  ;[1 1 1] Spawn the actual projectile. It inherits the explosion effects container from the arrow
  (bind _returnvalue (Projectile|SpawnProjectile :ProjectileClass (Variables|Default|GetClusterBombClass) :RequestedBy (Variables|Actor|GetInstigator) :SpawnLocation (Transformation|GetActorLocation) :SpawnRotation "0, 0, 0" :SpawnDirection (CallFunction|GetSpawnDirectionasRotation (Transformation|GetActorLocation) (LocalVariables|GetSpawnDirection)) :EffectContainerSpecToApplyOnExplode (Variables|Delivery|GetEffectContainerSpecToApplyOnExplode) :bUseDefaultPhysics false :InitialSpeed (CallFunction|GetBombSpeedfromRange) :GravityScale 1.0 :ChargePercent 0.0))
  (bind _asb_prj_bow_cluster_bomb_secondary_athena (Utilities|Casting|CastToB_Prj_Bow_ClusterBomb_Secondary_Athena _returnvalue)
    (:then
      (bind _output_get_2 (Variables|Default|SetPrevBombSpawned _asb_prj_bow_cluster_bomb_secondary_athena)))
    (:CastFailed
      _))
  (Math|Integer|IncrementInt (Variables|Default|GetNumBombsSpawned)
    (:%20%20
      (if (CallFunction|CheckAllBombsSpawned)
        (Class|BPrjBowClusterBombSecondaryAthena|SetLastBomb _output_get_2)))))

(fn ConstructionScript
  (|CallParentFunction))
