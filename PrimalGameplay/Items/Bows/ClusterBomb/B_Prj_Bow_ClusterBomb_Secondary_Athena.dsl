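; Secondary cluster bomb projectile for the Primal cluster bomb bow that explodes on a timed fuse.
; BeginPlay calls the parent, then sets a timer by name for DoExplosion, which is not defined in this graph.
; GetExplosionDelay returns either a random value between the min and max scalable-float delays or just the min, based on bUseRandomDelay.
; SetLastBomb swaps in a distinct explosion sound cue for the last bomb.
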(event EventBeginPlay
  (|CallParentFunction)
  (Utilities|Time|SetTimerbyFunctionName 0 "DoExplosion" (CallFunction|GetExplosionDelay)))

(event Custom|SetLastBomb
  (Variables|Sounds|SetCue_ExplosionSound (Variables|Default|GetCue_LastExplosionSound)))

(fn GetExplosionDelay
  (return (select (Variables|Default|GetbUseRandomDelay) (Math|Random|RandomFloatinRange (ScalableFloat|GetValueatLevel (Variables|Default|GetMinExplodeDelay)) (ScalableFloat|GetValueatLevel (Variables|Default|GetMaxExplodeDelay))) (ScalableFloat|GetValueatLevel (Variables|Default|GetMinExplodeDelay)))))

(fn ConstructionScript
  (|CallParentFunction))
